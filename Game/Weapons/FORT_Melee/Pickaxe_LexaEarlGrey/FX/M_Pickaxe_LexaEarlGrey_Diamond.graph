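# Translucent unlit Niagara sprite material for the LexaEarlGrey pickaxe FX, named as a diamond shape with an optional ring.
# Particle color RGB feeds MF_FX_EmissiveAdjustments, and particle alpha multiplies the procedural mask.
# The mask inverts centered, absolute, powered UVs, but doubles the R component alone, leaving the Squash-warped V unused.
# A static switch adds a SphereMask ring built from distance to UV center.
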Material /Game/Weapons/FORT_Melee/Pickaxe_LexaEarlGrey/FX/M_Pickaxe_LexaEarlGrey_Diamond
graph {
 properties {bUsedWithNiagaraSprites=true, blendMode=BLEND_Translucent, pixelDepthOffsetMode=PDOM_Legacy, shadingModel=MSM_Unlit, usedShadingModels=Unlit}
 nodes [
  n1 = ParticleColor()
  MaterialFunctionCall(Emissive=@n1.RGB) {materialFunction=&MF_FX_EmissiveAdjustments}
  n2 = TextureCoordinate()
  n3 = ComponentMask(None=@n2) {g=true}
  n4 = ScalarParameter() {defaultValue=1, parameterName=Squash, sliderMax=2}
  n5 = ConstantBiasScale(None=AppendVector(A=ComponentMask(None=@n2) {r=true}, B=Power(Base=@n3, Exp=@n4) {constExponent=0.6})) {bias=-0.5, scale=2}
  n6 = MaterialFunctionCall(Float2=Power(Base=Abs(None=@n5), Exp=ScalarParameter() {parameterName=Power, sliderMax=10})) {materialFunction=&BreakOutFloat2Components}
  n7 = Add(A=@n6.R, B=@n6.R)
  n8 = SphereMask(
   A=Distance(A=TextureCoordinate(), B=Constant() {r=0.5})
   B=ScalarParameter() {defaultValue=0.4, parameterName=Ring Size, sliderMax=0.5}
   Radius=ScalarParameter() {defaultValue=0.01, parameterName=Ring Thickness, sliderMax=1}
   Hardness=ScalarParameter() {defaultValue=0.1, parameterName=Ring Hardness, sliderMax=1}
  ) {hardnessPercent=0}
  n9 = Constant()
  Multiply(A=@n1.A, B=Add(A=OneMinus(None=@n7), B=StaticSwitchParameter(True=@n8, False=@n9) {parameterName=Add Ring}))
 ]
}
